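# Default Lit material for Niagara mesh particles, by name the candy bit effects of the LacedCrimp pickaxe.
# A camera-to-vertex-normal dot product blends two color parameters and masks a third as a rim tint, giving a cheap textureless Fresnel look.
# Particle color tints both the blend alone and the rimmed result, which feeds an emissive adjustment function.
# Scalar parameters supply specular, roughness and metallic values.
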Material /BRCosmetics/Weapons/FORT_Melee/Pickaxe_LacedCrimp/FX/M_Pickaxe_LacedCrimp_CandyBits
graph {
 properties {bAutomaticallySetUsageInEditor=false, bUsedWithNiagaraMeshParticles=true, usedShadingModels=Default Lit}
 nodes [
  n1 = DotProduct(A=CameraVectorWS(), B=VertexNormalWS())
  n2 = LinearInterpolate(
   A=VectorParameter() {parameterName=Color_1, defaultValue=<rgba 0.3698 0.0366 0.0796 1>}.RGB
   B=VectorParameter() {parameterName=Color_2, defaultValue=<rgba 0.1981 0.0185 0.0467 1>}.RGB
   Alpha=Saturate(None=Power(Base=@n1) {constExponent=3})
  )
  n3 = Multiply(A=@n1) {constB=2.7}
  n4 = Multiply(A=VectorParameter() {parameterName=Color_3, defaultValue=<rgba 0.4531 0.0189 0.1015 1>}.RGB, B=Saturate(None=OneMinus(None=@n3)))
  n5 = ParticleColor()
  MaterialFunctionCall(Emissive=Multiply(A=Add(A=@n2, B=@n4), B=@n5.RGB)) {materialFunction=&MF_FX_EmissiveAdjustments}
  Multiply(A=@n2, B=@n5.RGB)
  ScalarParameter() {defaultValue=0.6, parameterName=Specular}
  ScalarParameter() {defaultValue=0.1, parameterName=Roughness}
  ScalarParameter() {defaultValue=0.2, parameterName=Metallic}
 ]
}
