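# Character-parent material for the LacedCrimp pickaxe blade, faking a glowing, bubbly, refractive interior.
# A static switch selects either a frontface layer with opacity controls or a backface layer masked by TwoSidedSign.
# Glow blends two colours through a local-space sphere mask, enriched by refracted skylight, sun scattering and night dimming.
# Bump-offset UVs, bubble normals, a noise mask and scene-depth thickness supply depth and opacity detail.
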Material /BRCosmetics/Weapons/FORT_Melee/Pickaxe_LacedCrimp/Materials/M_LacedCrimp_Blade
graph {
 properties {
  bUseMaterialAttributes=true
  bUsedWithMeshDeformer=true
  bUsedWithMorphTargets=true
  bUsedWithSkeletalMesh=true
  pixelDepthOffsetMode=PDOM_Legacy
  translucencyLightingMode=TLM_SurfacePerPixelLighting
  usedShadingModels=Default Lit
 }
 nodes [
  CameraVectorWS()
  n1 = MaterialFunctionCall() {materialFunction=&MF_CharacterParent_2023_PreFX}
  MaterialFunctionCall(
   Material=StaticSwitchParameter(True=NamedRerouteUsage(), False=NamedRerouteUsage()) {group=LacedCrimp, parameterName=IsFrontFace}
   M=@n1.M
   Pre Skinned Local Position=@n1.M
   Distance Blend=@n1.M
   Fresnel=@n1.M
  ) {materialFunction=&MF_CharacterParent_2023_PostFX}
  n2 = NamedRerouteUsage()
  n3 = ScalarParameter() {defaultValue=17, group=LacedCrimp, parameterName=BubbleTile}
  NamedRerouteDeclaration(None=TextureSample(UVs=Multiply(A=@n2, B=@n3)) {samplerType=SAMPLERTYPE_Normal, texture=/BRCosmetics/Weapons/FORT_Melee/Pickaxe_LacedCrimp/Textures/T_LacedCrimpBubbles_N}.RGB) {name=BubbleNormal, nodeColor=<rgba 0.0078 1 0 1>}
  n4 = GetMaterialAttributes(None=NamedRerouteUsage()) {attributeGetTypes [#1]}
  n5 = TwoSidedSign()
  n6 = SetMaterialAttributes!(
   MaterialAttributes=@n4.MaterialAttributes
   Base Color=Constant()
   Emissive Color=NamedRerouteUsage()
   Roughness=Add(
    A=NamedRerouteDeclaration(None=@n4.Roughness) {name=IN_Roughness, nodeColor=<rgba 1 0 0.6797 1>}
    B=ScalarParameter() {defaultValue=0.464, group=LacedCrimp, parameterName=BackfaceRoughnessBias}
   )
   Opacity Mask=Saturate(None=Multiply(A=@n5) {constB=-1})
  ) {attributeSetTypes [
    #2
    #3
    #1
    #4
   ]}
  NamedRerouteDeclaration(None=@n6) {name=BackfaceMaterial, nodeColor=<rgba 0 1 1 1>}
  n7 = SceneDepth()
  n8 = MaterialFunctionCall!(World Position) {materialFunction=/Engine/Functions/Engine_MaterialFunctions02/Utility/DepthFromWorldPosition}
  n9 = Saturate(None=Divide(A=Subtract(A=@n7.R, B=@n8.Result), B=ScalarParameter() {defaultValue=2, group=LacedCrimp, parameterName=DepthThickness}))
  n10 = ScalarParameter() {defaultValue=0.5, group=LacedCrimp, parameterName=DepthExponent}
  NamedRerouteDeclaration(None=Min(A=Power(Base=@n9, Exp=@n10), B=ScalarParameter() {defaultValue=0.4, group=LacedCrimp, parameterName=DepthMaxOpacity})) {name=DepthOpacity, nodeColor=<rgba 1 0 0.8906 1>}
  n11 = SetMaterialAttributes!(
   MaterialAttributes=NamedRerouteUsage()
   World Position Offset=Multiply(A=VertexNormalWS()) {constB=0.001}
   Specular=ScalarParameter() {defaultValue=1, group=LacedCrimp, parameterName=FrontfaceSpecular}
   Roughness=Add(A=NamedRerouteUsage(), B=ScalarParameter() {defaultValue=0.144, group=LacedCrimp, parameterName=FrontfaceRoughnessBias})
   Metallic=Constant()
   Opacity=NamedRerouteUsage()
   Base Color=NamedRerouteUsage()
   Emissive Color=NamedRerouteUsage()
  ) {attributeSetTypes [
    #5
    #6
    #1
    #7
    #8
    #2
    #3
   ]}
  NamedRerouteDeclaration(None=@n11) {name=FrontfaceMaterial, nodeColor=<rgba 0 1 0.8359 1>}
  n12 = NamedRerouteUsage()
  n13 = ScalarParameter() {defaultValue=0.2, group=LacedCrimp, parameterName=MaskOpacityMin}
  n14 = ScalarParameter() {defaultValue=1, group=LacedCrimp, parameterName=MaskOpacityMax}
  n15 = NamedRerouteUsage()
  n16 = ScalarParameter() {defaultValue=3, group=LacedCrimp, parameterName=OpacityMaskTile}
  n17 = NamedRerouteDeclaration(None=TextureSample(UVs=Multiply(A=@n15, B=@n16)) {samplerType=SAMPLERTYPE_Grayscale, texture=/Game/Global/Textures/Default/Noise/LowResBlurredNoiseNormalized}.R) {name=OpacityMask, nodeColor=<rgba 0.9688 1 0 1>}
  n18 = Add(A=Multiply(A=NamedRerouteUsage(), B=LinearInterpolate(A=@n13, B=@n14, Alpha=@n17)), B=NamedRerouteUsage())
  NamedRerouteDeclaration(None=Multiply(A=ComponentMask(None=@n12) {b=true}, B=Saturate(None=@n18))) {name=FrontfaceOpacity, nodeColor=<rgba 1 0.5156 0 1>}
  n19 = Desaturation(None=NamedRerouteUsage(), Fraction=ScalarParameter() {group=LacedCrimp, parameterName=FrontfaceDesaturation})
  NamedRerouteDeclaration(None=Multiply(A=@n19, B=ScalarParameter() {defaultValue=0.3, group=LacedCrimp, parameterName=FrontfaceEmissiveStrength})) {name=FrontfaceEmissive, nodeColor=<rgba 0.1953 0 1 1>}
  NamedRerouteDeclaration(None=Multiply(A=@n19, B=ScalarParameter() {defaultValue=0.3, group=LacedCrimp, parameterName=FrontfaceBaseColorStrength})) {name=FrontfaceBasecolor, nodeColor=<rgba 0.1953 0 1 1>}
  NamedRerouteDeclaration(None=MaterialFunctionCall!(
    UV Coord=TextureCoordinate()
    Height=ScalarParameter() {defaultValue=1.2968, group=LacedCrimp, parameterName=UVOffsetHeight}
    ReferencePlane
    HeightRatio
    Normal
   ) {materialFunction=/Game/Global/MaterialFunctions/MF_BumpOffset_Accurate}.Result) {name=BumpOffsetUVs, nodeColor=<rgba 1 0.75 0 1>}
  n20 = AtmosphericLightVector()
  n21 = PixelNormalWS()
  n22 = Multiply(A=Multiply(A=DotProduct(A=@n20, B=@n21)) {constB=2}, B=@n21)
  n23 = DotProduct(
   A=Normalize(VectorInput=Subtract(A=@n22, B=@n20))
   B=Multiply(A=MaterialFunctionCall() {materialFunction=&CameraDirectionVector}.Result) {constB=-1}
  )
  n24 = Multiply(
   A=Power(Base=Max(A=@n23) {constB=0}, Exp=ScalarParameter() {defaultValue=2, group=LacedCrimp, parameterName=InnerScatteringPower})
   B=ScalarParameter() {defaultValue=1, group=LacedCrimp, parameterName=InnerScatteringScale}
  ) {constB=0.4}
  n25 = Saturate(None=DotProduct(A=MaterialFunctionCall() {materialFunction=&CameraDirectionVector}.Result, B=@n20))
  n26 = Multiply(A=@n25, B=ScalarParameter() {defaultValue=0.25, group=LacedCrimp, parameterName=SunScatterStrength})
  NamedRerouteDeclaration(None=Saturate(None=Add(A=@n24, B=@n26))) {name=InnerScatter, nodeColor=<rgba 1 0 0.1406 1>}
  NamedRerouteDeclaration(None=VectorParameter() {group=LacedCrimp, parameterName=GlowOrigin, defaultValue=<rgba 0 0 32 1>}.RGB) {name=GlowOrigin, nodeColor=<rgba 0.2188 0 1 1>}
  NamedRerouteDeclaration(None=@n1.Tech Art Mask) {name=TechArtMask, nodeColor=<rgba 0.1016 1 0 1>}
  n27 = MaterialFunctionCall(Float3=NamedRerouteUsage()) {materialFunction=&BreakOutFloat3Components}
  n28 = MaterialFunctionCall(X=@n27.R, Y=Constant(), Z=@n27.R) {materialFunction=&MakeFloat3}
  n29 = ScalarParameter() {defaultValue=10, group=LacedCrimp, parameterName=BubbleRefractionStrength}
  n30 = NamedRerouteUsage()
  n31 = NamedRerouteUsage()
  n32 = ScalarParameter() {defaultValue=15.8803, group=LacedCrimp, parameterName=BubbleMaskRadium}
  n33 = ScalarParameter() {defaultValue=0.176, group=LacedCrimp, parameterName=BubbleMaskHardness}
  NamedRerouteDeclaration(None=Multiply(A=Multiply(A=@n28.Result, B=@n29), B=SphereMask(A=@n30, B=@n31, Radius=@n32, Hardness=@n33))) {name=Bubbles, nodeColor=<rgba 1 0 0.4219 1>}
  NamedRerouteDeclaration(None=@n1.Pre skinned Local Position) {name=PreskinnedPos, nodeColor=<rgba 0 0.7422 1 1>}
  NamedRerouteDeclaration(None=@n1.Result) {name=IN_Material, nodeColor=<rgba 1 0.9375 0 1>}
  n34 = Multiply(A=CameraVectorWS()) {constB=-1}
  n35 = PixelNormalWS()
  n36 = ScalarParameter() {defaultValue=1.5, group=LacedCrimp, parameterName=SkylightRefraction}
  NamedRerouteDeclaration(None=SkyLightEnvMapSample(
    Direction=MaterialFunctionCall!(Ray Direction=@n34, Surface Normal=@n35, Refractive Index Origin, Refractive Index Target=@n36) {materialFunction=/Engine/Functions/Engine_MaterialFunctions01/Vectors/Refract}.Vector
    Roughness=ScalarParameter() {defaultValue=0.3, group=LacedCrimp, parameterName=SkylightRoughness}
   )) {name=Skylight, nodeColor=<rgba 0 1 0.4844 1>}
  n37 = Constant() {r=1}
  n38 = NamedRerouteUsage()
  n39 = NamedRerouteUsage()
  n40 = Power(
   Base=SphereMask(
    A=Add(A=@n38, B=@n39)
    B=NamedRerouteUsage()
    Radius=ScalarParameter() {defaultValue=18.0482, group=LacedCrimp, parameterName=GlowRadius}
    Hardness=ScalarParameter() {group=LacedCrimp, parameterName=GlowHardness}
   )
   Exp=ScalarParameter() {defaultValue=2, group=LacedCrimp, parameterName=GlowMaskExponent}
  )
  n41 = Multiply(
   A=MaterialFunctionCall(
    Value_Morning=@n37
    Value_Day=@n37
    Value_Evening=@n37
    Value_Night=ScalarParameter() {defaultValue=0.8, parameterName=NightEmissiveMulitplier}
   ) {materialFunction=&Scalar_Time_of_Day}.Result
   B=LinearInterpolate(
    A=VectorParameter() {group=LacedCrimp, parameterName=ColorB, defaultValue=<rgba 0.2396 0 0.1011 1>}.RGB
    B=VectorParameter() {group=LacedCrimp, parameterName=ColorA, defaultValue=<rgba 1 0 0.4849 1>}.RGB
    Alpha=NamedRerouteDeclaration(None=@n40) {name=GlowMask, nodeColor=<rgba 0 0.7422 1 1>}
   )
  )
  n42 = NamedRerouteUsage()
  n43 = Add(A=@n41, B=Multiply(A=Multiply(A=@n41, B=@n42), B=ScalarParameter() {defaultValue=1, group=LacedCrimp, parameterName=SkylightStrength}))
  n44 = ScalarParameter() {defaultValue=2, group=LacedCrimp, parameterName=ScatteringStrength}
  NamedRerouteDeclaration(None=LinearInterpolate(A=@n43, B=Multiply(A=@n43, B=@n44), Alpha=NamedRerouteUsage())) {name=BaseEmissive, nodeColor=<rgba 0.8281 1 0 1>}
  OneMinus(None=@n25)
  TextureSample() {samplerType=SAMPLERTYPE_Normal, texture=/Game/Global/Materials/MaterialGym/Materials/Bubbles/Texture/T_SlurpBubble_Global_N}
 ]
}
