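# Physics asset for the LacedCrimp two-handed pickaxe cosmetic, giving its dangling loops and strands simulated secondary motion.
# Capsule bodies cover the handle, small base anchor bones, two dynamic loops and two three-segment strands.
# Swing- and twist-limited constraints tie each loop to its base and chain strand segments from their bases.
# The first segment of strand B is constrained to both the handle and its base bone.
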PhysicsAsset /BRCosmetics/Weapons/FORT_Melee/Pickaxe_LacedCrimp/Meshes/LacedCrimp_Axe_Physics
bodies {
 Melee_TwoHanded_Handle [
  {shapeType=Capsule, length=52.6815, radius=2.6531, center=<-0.3829 0 6.1293>, rotation=<rot 0 -90 0>}
  {
   shapeType=Capsule
   length=3.5
   radius=2.5
   shapeName=Melee_TwoHanded_Handle_capsule
   center=<0.2822 -1.2948 -20.633>
   rotation=<rot -0.0002 0 90>
  }
 ]
 base_loop_A [{shapeType=Capsule, length=0, radius=0.5, shapeName=base_loop_A_capsule}]
 base_loop_B [{shapeType=Capsule, length=0, radius=0.5, shapeName=base_loop_B_capsule}]
 base_strand_A [{shapeType=Capsule, length=0, radius=0.5, shapeName=base_strand_A_capsule}]
 base_strand_B [{shapeType=Capsule, length=0, radius=0.5, shapeName=base_strand_B_capsule}]
 dyn_loop_A [{shapeType=Capsule, length=1.6998, radius=2, shapeName=dyn_loop_A_capsule, center=<5.5284 0.1004 -0.1049>, rotation=<rot 0 -92.6278 -86.8457>}]
 dyn_loop_B [{
   shapeType=Capsule
   length=2.815
   radius=2.3751
   shapeName=dyn_loop_B_capsule
   center=<3.5297 -0.3256 0.0365>
   rotation=<rot -0.6161 84.737 -86.5044>
  }]
 dyn_strand_A_0 [{
   shapeType=Capsule
   length=5.505
   radius=0.505
   shapeName=dyn_strand_A_0_capsule
   center=<3.1259 0.0334 0.005>
   rotation=<rot 0 89.8356 -89.29>
  }]
 dyn_strand_A_1 [{
   shapeType=Capsule
   length=5.005
   radius=0.505
   shapeName=dyn_strand_A_1_capsule
   center=<2.6513 0.0253 0.019>
   rotation=<rot 0 -90.1799 -89.1816>
  }]
 dyn_strand_A_2 [{
   shapeType=Capsule
   length=1.505
   radius=0.505
   shapeName=dyn_strand_A_2_capsule
   center=<1.02 0.0349 -0.048>
   rotation=<rot 0 -89.7488 -88.2331>
  }]
 dyn_strand_B_0 [{
   shapeType=Capsule
   length=4.7969
   radius=0.7532
   shapeName=dyn_strand_B_0_capsule
   center=<3.3722 0.0107 -0.0268>
   rotation=<rot 0 89.516 -90>
  }]
 dyn_strand_B_1 [{
   shapeType=Capsule
   length=4.7003
   radius=0.7288
   shapeName=dyn_strand_B_1_capsule
   center=<2.5065 0.0101 -0.0045>
   rotation=<rot 0 89.9506 -90>
  }]
 dyn_strand_B_2 [{
   shapeType=Capsule
   length=1.1496
   radius=0.7455
   shapeName=dyn_strand_B_2_capsule
   center=<0.8821 0.0294 -0.0427>
   rotation=<rot 0 -89.6921 -90>
  }]
}
constraints [
 {bone1Name=dyn_loop_B, bone2Name=base_loop_B, swing1LimitDegrees=30, swing2LimitDegrees=40, twistLimitDegrees=5}
 {bone1Name=dyn_loop_A, bone2Name=base_loop_A, swing1LimitDegrees=20, swing2LimitDegrees=30, twistLimitDegrees=5}
 {bone1Name=dyn_strand_A_0, bone2Name=base_strand_A, swing1LimitDegrees=35, swing2LimitDegrees=45, twistLimitDegrees=5}
 {bone1Name=dyn_strand_A_1, bone2Name=dyn_strand_A_0, swing1LimitDegrees=40, swing2LimitDegrees=45, twistLimitDegrees=5}
 {bone1Name=dyn_strand_A_2, bone2Name=dyn_strand_A_1, swing1LimitDegrees=20, swing2LimitDegrees=45, twistLimitDegrees=5}
 {bone1Name=dyn_strand_B_0, bone2Name=Melee_TwoHanded_Handle, swing1LimitDegrees=40, swing2LimitDegrees=45, twistLimitDegrees=5}
 {bone1Name=dyn_strand_B_1, bone2Name=dyn_strand_B_0, swing1LimitDegrees=35, swing2LimitDegrees=50, twistLimitDegrees=7}
 {bone1Name=dyn_strand_B_2, bone2Name=dyn_strand_B_1, swing1LimitDegrees=45, swing2LimitDegrees=50, twistLimitDegrees=10}
 {bone1Name=dyn_strand_B_0, bone2Name=base_strand_B, swing1LimitDegrees=40, swing2LimitDegrees=45, twistLimitDegrees=5}
]
